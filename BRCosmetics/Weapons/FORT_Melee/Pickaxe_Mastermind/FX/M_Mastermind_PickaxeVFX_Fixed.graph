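# Masked, two-sided lit material for Niagara mesh particles in the Mastermind pickaxe VFX, based on the character's MF_TwistedGoo function.
# It adds vertex-normal displacement from an actor-relative, world-aligned panning texture, weighted by inverted vertex red.
# Dynamic parameters drive pan speed, texture size and offset strength.
# Particle alpha replaces the opacity mask, and the attributes pass through a shared emissive adjustment function.
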Material /BRCosmetics/Weapons/FORT_Melee/Pickaxe_Mastermind/FX/M_Mastermind_PickaxeVFX_Fixed
graph {
 properties {
  bUseMaterialAttributes=true
  bUsedWithNiagaraMeshParticles=true
  blendMode=BLEND_Masked
  pixelDepthOffsetMode=PDOM_Legacy
  twoSided=true
  usedShadingModels=Default Lit
 }
 nodes [
  n1 = GetMaterialAttributes(None=MaterialFunctionCall!(In, Mask, isVehicle, isMech) {materialFunction=/Game/Characters/Player/Male/Medium/Bodies/M_MED_Mastermind/Materials/MF_TwistedGoo}.Result) {attributeGetTypes [#1]}
  n2 = VertexNormalWS()
  n3 = DynamicParameter() {defaultValue=<rgba 0 0.2 1 50>, paramNames [Panner Speed Scale, Mesh Selection, WPO_Offset, TextureSize]}
  n4 = DynamicParameter() {defaultValue=<rgba 0 0.2 1 50>, paramNames [Panner Speed Scale, Mesh Selection, WPO_Offset, TextureSize]}
  n5 = VectorParameter() {parameterName=Panning Speed, defaultValue=<rgba 0 0.1 0 0>}
  n6 = MaterialFunctionCall!(
   TextureObject=TextureObject() {texture=&2ndInfectedMapLinear}
   TextureSize=AppendVector(A=AppendVector(A=@n4.TextureSize, B=@n4.TextureSize), B=@n4.TextureSize)
   WorldPosition=Subtract(A=ActorPositionWS(), B=WorldPosition().XYZ)
   Export Float 4
   In=Multiply(A=ComponentMask(None=@n5.RGB) {g=true, r=true}, B=@n4.Panner Speed Scale)
   World Space Normal
   ProjectionTransitionContrast
  ) {materialFunction=/Game/Athena/SafeZone/WorldAlignedTexture_ScalePan}
  n7 = LinearInterpolate(B=@n3.WPO_Offset, Alpha=ComponentMask(None=@n6.XYZ Texture) {b=true})
  n8 = Multiply(A=OneMinus(None=VertexColor().R), B=Multiply(A=Multiply(A=@n2, B=@n7), B=@n6.XYZ Texture))
  MaterialFunctionCall!(In=SetMaterialAttributes!(MaterialAttributes=@n1.MaterialAttributes, Opacity Mask=ParticleColor().A, World Position Offset=Add(A=@n1.World Position Offset, B=@n8)) {attributeSetTypes [#2, #1]}) {materialFunction=/Game/Effects/Niagara/MasterMaterials/Functions/MF_FX_EmissiveAdjustments_MA}
 ]
}
